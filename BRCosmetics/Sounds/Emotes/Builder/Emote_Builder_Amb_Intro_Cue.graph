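# Sound cue for the intro audio of the Builder emote, classed as emote music.
# A mixer layers the ambient intro wave at reduced volume with the main intro wave, which is boosted and starts after a short fixed delay.
# Playback uses Athena emote attenuation and the emote foley concurrency set, and has a send to the music audio-analysis submix.
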SoundCue /BRCosmetics/Sounds/Emotes/Builder/Emote_Builder_Amb_Intro_Cue
root=Mixer_0
nodes [
 {id=Delay_0, children [WavePlayer_1], properties {delayMax=0.25, delayMin=0.25}}
 {id=Mixer_0, children [Delay_0,WavePlayer_0], properties {inputVolume [1.25,0.5]}}
 {id=WavePlayer_0, properties {soundWaveAssetPtr=&/Emote_Builder_Amb_Intro}}
 {id=WavePlayer_1, properties {soundWaveAssetPtr=&/Emote_Builder_Intro}}
]
properties {
 attenuationSettings=&Fort_Emotes_Athena_Att
 maxDistance=1600
 soundClassObject=&Fort_Music_Emote
 volumeMultiplier=0.5
 concurrencyOverrides {maxCount=3, resolutionRule=StopOldest}
 concurrencySet [/Game/Sounds/Concurrency_Rules/Athena_EmoteFoley_Conc]
 soundSubmixSends [{minSendLevel=0, soundSubmix=&Fortnite_AudioAnalysis_Music}]
}
