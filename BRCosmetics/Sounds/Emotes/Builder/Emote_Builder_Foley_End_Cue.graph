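# Sound cue for the ending foley of the Builder emote, set up with emote foley class, attenuation and concurrency settings.
# A mixer layers a randomly chosen crunch (one of two waves, pitch-varied) with a quieter foley wave that follows after a short fixed delay.
# The delayed foley varies in both pitch and volume on each play.
# The cue also has a send to the music audio-analysis submix.
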SoundCue /BRCosmetics/Sounds/Emotes/Builder/Emote_Builder_Foley_End_Cue
root=Mixer_0
nodes [
 {id=Delay_0, children [Modulator_1], properties {delayMax=0.75, delayMin=0.75}}
 {id=Mixer_0, children [Modulator_0,Delay_0], properties {inputVolume [1,0.65]}}
 {id=Modulator_0, children [Random_0], properties {pitchMax=1.03, pitchMin=0.94}}
 {id=Modulator_1, children [WavePlayer_1], properties {pitchMax=1.03, pitchMin=0.98, volumeMax=1, volumeMin=0.3}}
 {id=Random_0, children [WavePlayer_2,WavePlayer_3], properties {weights [1,1]}}
 {id=WavePlayer_1, properties {soundWaveAssetPtr=&/Emote_Builder_Foley}}
 {id=WavePlayer_2, properties {soundWaveAssetPtr=&/Emote_Builder_Crunch}}
 {id=WavePlayer_3, properties {soundWaveAssetPtr=&/Emote_Builder_Crunch_02}}
]
properties {
 attenuationSettings=&Fort_Emotes_Athena_Att
 audioPropertiesSheet=/BRCosmetics/Sounds/Settings/PropertySheets/Emotes/APS_Cosmetic_Emote_Foley
 maxDistance=1600
 soundClassObject=/Game/Sounds/FortSoundClassesAndModes/Fort_Foley_Emote
 concurrencyOverrides {maxCount=3, resolutionRule=StopOldest}
 concurrencySet [/Game/Sounds/Concurrency_Rules/Athena_EmoteFoley_Conc]
 soundSubmixSends [{minSendLevel=0, soundSubmix=&Fortnite_AudioAnalysis_Music}]
}
